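# Niagara swing trail for the SilverCanine pickaxe cosmetic, with six looks selected by User.Variant.
# Three ribbon emitters use shared AnimTrail modules: a variant-tinted trail, a thin sharp streak, and a delayed, variant-tinted dark smoke ribbon.
# A GPU sprite emitter samples the skeletal mesh, using variant-indexed arrays and switches for spawn rate, size, color and renderer visibility tag; two variants spawn none.
# All layers fade with System.TrailAlphaFade, most also scaling by time of day.
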NiagaraSystem /BRCosmetics/Weapons/FORT_Melee/Pickaxe_SilverCanine/FX/NS_Pickaxe_SilverCanine_Trail
dependencies {
 usedDataInterfaces [
  &NiagaraDataInterfaceCurve
  &NiagaraDataInterfaceArrayColor
  &NiagaraDataInterfaceArrayFloat
  &NiagaraDataInterfaceArrayInt32
  &NiagaraDataInterfaceColorCurve
  &NiagaraDataInterfaceSocketReader
 ]
 usedDynamicInputs []
 usedModules [
  &self:SystemSpawnScript.SystemScriptSource.SystemScriptGraph.NiagaraNodeAssignment_1.SetVariables_#1
  &SystemState
  &self:SystemSpawnScript.SystemScriptSource.SystemScriptGraph.NiagaraNodeAssignment_3.SetVariables_#2
 ]
}
emitters {
 Dark_Smoke001 {
  emitterUpdateScript [
   &EmitterState(Life Cycle Mode=::NewEnumerator1, Loop Behavior=::NewEnumerator1, Loop Delay=0.17, Loop Duration=0.1, UseLoopDelay=true)
   &FN_SpawnAnimTrail()
  ]
  particleSpawnScript [
   &FN_InitializeAnimTrailBoneIndecies() {module=FN_InitializeAnimTrailBoneIndecies001}
   &FN_InitializeAnimTrail(Lifetime=0.35, Position Offset=&TransformVector(Vector=<0 0 -5>, Destination Space=::Simulation)) {module=FN_InitializeAnimTrail001}
  ]
  particleUpdateScript [
   &FN_UpdateAnimTrail()
   &ScaleRibbonWidth(Ribbon Width Scale=&FloatFromCurve(FloatCurve=DataInterface() {propertyValues {Curve {keys [{arriveTangent=-0.8703, time=0.2, value=1}, <at 1 0.8>]}}}))
   &ScaleColor(
    Scale Alpha=&Multiply_Float(
     A=&FloatFromCurve(
      FloatCurve=DataInterface() {propertyValues {Curve {keys [{arriveTangent=19.9617, interpMode=RCIM_Cubic, tangentMode=RCTM_User, time=0, value=1}, <at 1 0>]}}}
      Scale Curve=&FloatFromCurve(
       FloatCurve=DataInterface() {propertyValues {Curve {
          defaultValue=1
          keys [{interpMode=RCIM_Cubic, tangentMode=RCTM_User, time=0.5, value=1}, {interpMode=RCIM_Cubic, tangentMode=RCTM_User, time=1, value=0}]
         }}}
       CurveIndex=&RemapRange(Input Value To Remap=$Emitter.Age, Input Min=0, Input Max=0.35, Output Min=0, Output Max=1)
       Scale Curve=0.5
      )
     )
     B=&FloatFromCurve(
      FloatCurve=DataInterface() {propertyValues {Curve {
         defaultValue=1
         keys [
          {interpMode=RCIM_Cubic, tangentMode=RCTM_User, time=0, value=1.3}
          {interpMode=RCIM_Cubic, tangentMode=RCTM_User, time=0.5, value=1}
          {tangentMode=RCTM_User, time=1, value=1.3}
         ]
        }}}
      CurveIndex=Linked() {linkedVariable {name=NPC.FortniteNPC.TimeOfDay_0-1, type=NiagaraFloat}}
      Scale Curve=&Multiply_Float(A=1, B=$System.TrailAlphaFade)
     )
    )
    Scale RGB=&Lerp_Vector(
     A=&MakeVectorFromLinearColorRGB(Color=&SelectLinearColorFromArray(
       Array Sampling Mode=::NewEnumerator1
       Direct Array Index=&MakeIntFromFloat(Float=$User.Variant)
       Color Selection Array=DataInterface() {propertyValues {ColorData [<rgba 0.0003 0.0545 0.0423 1>, <rgba 0 0 0 1>, <rgba 0 0 0 1>, <rgba 0.2703 0.4169 0.4086 1>, <rgba 0.1677 0.0515 0.1908 1>, <rgba 0 0 0 3>]}}
       Random Seed=?
      ))
     B=&MakeVectorFromLinearColorRGB(Color=&HueShiftLinearColor(
       Hue=0
       Saturation=0.5
       Value=10
       Linear Color=&SelectLinearColorFromArray(
        Array Sampling Mode=::NewEnumerator1
        Direct Array Index=&MakeIntFromFloat(Float=$User.Variant)
        Color Selection Array=DataInterface() {propertyValues {ColorData [<rgba 0.0003 0.0545 0.0423 1>, <rgba 0 0 0 1>, <rgba 0 0 0 1>, <rgba 0.2703 0.4169 0.4086 1>, <rgba 0.1677 0.0515 0.1908 1>, <rgba 0 0 0 3>]}}
        Random Seed=?
       )
      ))
     Alpha=&FloatFromCurve(FloatCurve=DataInterface() {propertyValues {Curve {
         defaultValue=1
         keys [{interpMode=RCIM_Cubic, tangentMode=RCTM_User, time=0.4, value=0}, {interpMode=RCIM_Cubic, tangentMode=RCTM_User, time=1, value=1}]
        }}})
    )
    ScaleA=true
    ScaleRGB=true
   )
   &SpriteFacingAndAlignment(Sprite Facing=$Particles.RibbonFacing, WriteAlignment=false, WriteFacing=true)
  ]
  rendererClasses [&NiagaraRibbonRendererProperties]
  renderers [{rendererClass=&NiagaraRibbonRendererProperties}]
 }
 Trail {
  emitterSpawnScript [&FN_RibbonIDOffsetUtility()]
  emitterUpdateScript [
   &EmitterState()
   &FN_SpawnAnimTrail()
  ]
  particleSpawnScript [
   &FN_InitializeAnimTrailBoneIndecies()
   &FN_InitializeAnimTrail(Lifetime=0.2)
  ]
  particleUpdateScript [
   &FN_UpdateAnimTrail() {module=FN_UpdateAnimTrail001}
   &ScaleColor(
    Scale Alpha=&Multiply_Float(
     A=&FloatFromCurve(FloatCurve=DataInterface() {propertyValues {Curve {keys [{interpMode=RCIM_Cubic, tangentMode=RCTM_User, time=0, value=1.1}, {interpMode=RCIM_Cubic, tangentMode=RCTM_User, time=1, value=0}]}}})
     B=&FloatFromCurve(
      FloatCurve=DataInterface() {propertyValues {Curve {
         defaultValue=1
         keys [{interpMode=RCIM_Cubic, tangentMode=RCTM_User, time=0.7, value=1}, {interpMode=RCIM_Cubic, tangentMode=RCTM_User, time=1, value=0}]
        }}}
      CurveIndex=&RemapRange(Input Value To Remap=$Emitter.Age, Input Min=0, Input Max=0.2, Output Min=0, Output Max=1)
      Scale Curve=&Multiply_Float(A=1, B=$System.TrailAlphaFade)
     )
    )
    Scale RGB=&Lerp_Vector(
     A=&MakeVectorFromLinearColorRGB(Color=&SelectLinearColorFromArray(
       Array Sampling Mode=::NewEnumerator1
       Direct Array Index=&MakeIntFromFloat(Float=$User.Variant)
       Color Selection Array=DataInterface() {propertyValues {ColorData [
          <rgba 0 0 0 1>
          <rgba 0.3995 0.0245 0.0665 1>
          <rgba 0.4995 0.3316 0.0251 1>
          <rgba 0.0946 0.6244 0.7473 1>
          <rgba 0.3661 0.59 2 1>
          <rgba 0.1522 0.0635 1 1>
         ]}}
       Random Seed=?
      ))
     B=&MakeVectorFromLinearColorRGB(Color=&HueShiftLinearColor(
       Hue=0
       Saturation=0.5
       Value=10
       Linear Color=&SelectLinearColorFromArray(
        Array Sampling Mode=::NewEnumerator1
        Direct Array Index=&MakeIntFromFloat(Float=$User.Variant)
        Color Selection Array=DataInterface() {propertyValues {ColorData [
           <rgba 0 0 0 1>
           <rgba 0.3995 0.0245 0.0665 1>
           <rgba 0.4995 0.3316 0.0251 1>
           <rgba 0.0946 0.6244 0.7473 1>
           <rgba 0.3661 0.59 2 1>
           <rgba 0.1522 0.0635 1 1>
          ]}}
        Random Seed=?
       )
      ))
     Alpha=&FloatFromCurve(FloatCurve=DataInterface() {propertyValues {Curve {
         defaultValue=1
         keys [{interpMode=RCIM_Cubic, tangentMode=RCTM_User, time=0.4, value=0}, {interpMode=RCIM_Cubic, tangentMode=RCTM_User, time=1, value=1}]
        }}})
    )
    ScaleA=true
    ScaleRGB=true
   ) {module=ScaleColor001}
  ]
  rendererClasses [&NiagaraRibbonRendererProperties]
  renderers [{rendererClass=&NiagaraRibbonRendererProperties}]
 }
 TrailSharp {
  emitterSpawnScript [&FN_RibbonIDOffsetUtility()]
  emitterUpdateScript [
   &EmitterState()
   &FN_SpawnAnimTrail()
  ]
  particleSpawnScript [
   &FN_InitializeAnimTrailBoneIndecies(First Bone or Socket Index=1)
   &FN_InitializeAnimTrail(AnimTrailWidth=4, Color=<rgba 0.585 0.6229 0.65 1>, Lifetime=0.25)
  ]
  particleUpdateScript [
   &SolveForcesAndVelocity(Clamp Velocity=false, Limit Acceleration=false)
   &FN_UpdateAnimTrail() {module=FN_UpdateAnimTrail001}
   &ScaleColor(
    Scale Alpha=&Multiply_Float(
     A=&FloatFromCurve(
      FloatCurve=DataInterface() {propertyValues {Curve {keys [{interpMode=RCIM_Cubic, tangentMode=RCTM_User, time=0, value=1}, {interpMode=RCIM_Cubic, tangentMode=RCTM_User, time=1, value=0}]}}}
      Scale Curve=&FloatFromCurve(
       FloatCurve=DataInterface() {propertyValues {Curve {
          defaultValue=1
          keys [{interpMode=RCIM_Cubic, tangentMode=RCTM_User, time=0.5, value=1}, {interpMode=RCIM_Cubic, tangentMode=RCTM_User, time=1, value=0}]
         }}}
       CurveIndex=&RemapRange(Input Value To Remap=$Emitter.Age, Input Min=0, Input Max=0.25, Output Min=0, Output Max=1)
      )
     )
     B=&FloatFromCurve(
      FloatCurve=DataInterface() {propertyValues {Curve {
         defaultValue=1
         keys [
          {interpMode=RCIM_Cubic, tangentMode=RCTM_User, time=0, value=0.8}
          {interpMode=RCIM_Cubic, tangentMode=RCTM_User, time=0.5, value=1}
          {tangentMode=RCTM_User, time=1, value=0.8}
         ]
        }}}
      CurveIndex=Linked() {linkedVariable {name=NPC.FortniteNPC.TimeOfDay_0-1, type=NiagaraFloat}}
      Scale Curve=&Multiply_Float(A=1, B=$System.TrailAlphaFade)
     )
    )
    ScaleA=true
    ScaleRGB=false
   ) {module=ScaleColor001}
   &ScaleRibbonWidth(Ribbon Width Scale=&FloatFromCurve(FloatCurve=DataInterface() {propertyValues {Curve {keys [<at 0 0.5>,<at 0.1 1>,<at 1 0.3>]}}}))
  ]
  rendererClasses [&NiagaraRibbonRendererProperties]
  renderers [{rendererClass=&NiagaraRibbonRendererProperties}]
 }
 Variant_Sprites {
  simTarget=GPUComputeSim
  emitterUpdateScript [
   &EmitterState(
    Enable Distance Culling=true
    Enable Visibility Culling=true
    MaxDistance=2500
    Module.bApplySpawnCountScaleByDistance=true
    Module.bUseMaxDistance=true
    Module.bUseMinDistance=false
    Scalability Mode=::NewEnumerator1
    Scale Spawn Count=true
   )
   &SpawnRate(
    SpawnRate=&SelectFloatFromArray(
     Direct Array Index=&MakeIntFromFloat(Float=$User.Variant)
     Float Selection Array=DataInterface() {propertyValues {FloatData [0,0,80,80,50,80]}}
    )
    Use Spawn Probability=false
   )
  ]
  particleSpawnScript [
   /SetVariables(Particles.VisibilityTag=&SelectIntFromArray(
     Direct Array Index=&MakeIntFromFloat(Float=$User.Variant)
     Int Selection Array=DataInterface() {propertyValues {IntData [0,0,2,3,4,5]}}
    )) {isSetParameters=true}
   &InitializeParticle(
    Color=&MergeHelper_ColorSwitch(
     0=<rgba 0 0 0 0>
     1=<rgba 0 0 0 0>
     2=&RandomRangeLinearColor(Minimum=<rgba 10 3.6667 0.5 1>, Maximum=<rgba 10 5.5 1 1>)
     3=<rgba 0.7646 0.9519 1 1>
     4=<rgba 0 0 0 0>
     5=<rgba 0 0 0 0>
     Index=&MakeIntFromFloat(Float=$User.Variant)
     Num=::NewEnumerator4
    )
    Lifetime=&MergeHelper_FloatSwitch(
     0=0
     1=0
     2=0.4
     3=0.4
     4=0.4
     5=0.25
     Index=&MakeIntFromFloat(Float=$User.Variant)
     Num=::NewEnumerator4
    )
    Position Offset=&RandomRangeVector(Minimum=<-5 -5 60>, Maximum=<-5 5 20>)
    Position Offset Coordinate Space=::Local
    Sprite Rotation Mode=::NewEnumerator2
    Sprite Size Mode=::NewEnumerator3
    Uniform Sprite Size=&MergeHelper_FloatSwitch(
     0=0
     1=0
     2=&RandomRangeFloat(Minimum=3, Maximum=6)
     3=&RandomRangeFloat(Minimum=5, Maximum=6)
     4=&RandomRangeFloat(Minimum=7, Maximum=11)
     5=&RandomRangeFloat(Minimum=10, Maximum=15)
     Index=&MakeIntFromFloat(Float=$User.Variant)
     Num=::NewEnumerator4
    )
    UsePositionOffset=true
    Write Material Random=true
   )
   &AddVelocity(
    Module.Use Falloff Distance=false
    Override Seed=false
    Velocity Mode=::NewEnumerator1
    Velocity Speed=&RandomRangeFloat(Minimum=0, Maximum=50)
   ) {module=AddVelocity001}
   &SkeletalMeshLocation(
    MaskSpawnGroup=false
    Mesh Sampling Type=::NewEnumerator5
    Sampled Position Offset=&RandomRangeVector(Minimum=<0 0 0>, Maximum=<0 0 60>)
    Skeletal Mesh=$System.Skeletal Mesh
    Socket Sampling Mode=::NewEnumerator5
   )
  ]
  particleUpdateScript [
   &ParticleState()
   &SpriteRotationRate(Rotation Rate=&FloatFromCurve(
     FloatCurve=DataInterface() {propertyValues {Curve {keys [{interpMode=RCIM_Cubic, tangentMode=RCTM_User, time=0, value=1}, {interpMode=RCIM_Cubic, tangentMode=RCTM_User, time=1, value=0}]}}}
     Scale Curve=&Multiply_FloatByInt(
      Float=&RandomRangeFloat(Minimum=100, Maximum=200)
      Integer=&SelectIntFromArray(
       Array Sampling Mode=::NewEnumerator1
       Direct Array Index=?
       Int Selection Array=DataInterface() {propertyValues {IntData [-1,1]}}
       Random Seed=0
      )
     )
    ))
   &AccelerationForce(Acceleration=<0 0 -100>)
   &Drag(Drag=10, Use Linear Drag=true, Use Rotational Drag=true)
   &CurlNoiseForce(Noise Strength=200, Pan Noise=false)
   &SolveForcesAndVelocity(Clamp Velocity=false, Limit Acceleration=false)
   &ScaleSpriteSize(Uniform Curve Sprite Scale=DataInterface() {propertyValues {Curve {keys [
        {arriveTangent=-1.3862, interpMode=RCIM_Cubic, time=0, value=1}
        {arriveTangent=-1.8417, interpMode=RCIM_Cubic, leaveTangent=-1.8417, tangentMode=RCTM_User, time=1, value=0}
       ]}}})
   &ScaleColor(
    Scale Alpha=&Multiply_Float(
     A=&FloatFromCurve(
      FloatCurve=DataInterface() {propertyValues {Curve {keys [<at 0.3 1>,<at 1 0>]}}}
      Scale Curve=&Multiply_Float(A=$System.TrailAlphaFade, B=1)
     )
     B=&FNTimeOfDayScalar(Night Scale=0.5, Morning Scale=0.75, Evening Scale=0.75)
    )
    ScaleA=true
    ScaleRGB=false
   )
   /SetVariables(Particles.GlowColor=&MakeLinearColorFromVectorAndFloat(
     "Vector (RGB)"=&MakeVectorFromLinearColorRGB(Color=&MergeHelper_ColorSwitch(
       0=&Multiply_LinearColorByFloat(
        Linear Color=&ColorFromCurve(
         ColorCurve=DataInterface() {propertyValues {
           AlphaCurve {keys [<at 0 1 RCIM_Cubic>]}
           BlueCurve {keys [
             <at 0 1>
             <at 0.2788 1>
             {arriveTangent=0.4199, interpMode=RCIM_Cubic, leaveTangent=0.4199, time=0.5087, value=0.9509}
             <at 0.7781 1.3267>
             <at 1 1>
            ]}
           GreenCurve {keys [
             <at 0 0.6539>
             <at 0.2788 0.2522>
             {arriveTangent=1.0874, interpMode=RCIM_Cubic, leaveTangent=1.0874, time=0.5087, value=0.5156}
             <at 0.7781 1.5>
             <at 1 0.6539>
            ]}
           RedCurve {keys [
             <at 0 0.4301>
             <at 0.2788 0.613>
             {arriveTangent=-0.1655, interpMode=RCIM_Cubic, leaveTangent=-0.1655, time=0.5087, value=1}
             <at 0.7781 0.3013>
             <at 1 0.4301>
            ]}
          }}
         CurveIndex=&Frac_Float(Float=&Add_Float(A=$Particles.NormalizedAge, B=&RandomRangeFloat(Minimum=0, Maximum=-0.5)))
        )
        Float=&RandomRangeFloat(Minimum=0.2, Maximum=0.5)
       )
       1=&Multiply_LinearColorByFloat(
        Linear Color=&ColorFromCurve(
         ColorCurve=DataInterface() {propertyValues {
           AlphaCurve {keys [<at 0 1 RCIM_Cubic>]}
           BlueCurve {keys [
             <at 0.0284 0.8677>
             {arriveTangent=-0.4812, interpMode=RCIM_Cubic, leaveTangent=-0.4812, time=0.4737, value=0.8646}
             <at 0.7925 0.5>
             <at 0.9918 0.8646>
            ]}
           GreenCurve {keys [
             <at 0.0284 1>
             {arriveTangent=-1.2221, interpMode=RCIM_Cubic, leaveTangent=-1.2221, time=0.4737, value=0.1067}
             <at 0.7925 0.0662>
             <at 0.9918 0.7183>
            ]}
           RedCurve {keys [
             <at 0.0284 0>
             {arriveTangent=0.4161, interpMode=RCIM_Cubic, leaveTangent=0.4161, time=0.4737, value=0.231}
             <at 0.7925 0.3179>
             <at 0.9918 0>
            ]}
          }}
         CurveIndex=&Frac_Float(Float=&Add_Float(A=$Particles.NormalizedAge, B=&RandomRangeFloat(Minimum=0, Maximum=1)))
        )
        Float=&RandomRangeFloat(Minimum=0.2, Maximum=0.5)
       )
       2=?
       3=?
       Index=&MergeHelper_IntSwitch(0=0, 1=0, 2=0, 3=0, 4=0, 5=1, Index=&MakeIntFromFloat(Float=$User.Variant), Num=::NewEnumerator4)
       Num=::NewEnumerator0
      ))
     "Float (Alpha)"=&FloatFromCurve(
      FloatCurve=DataInterface() {propertyValues {Curve {keys [<at 0.5 1>,<at 1 1>]}}}
      Scale Curve=&Multiply_Float(A=$System.TrailAlphaFade, B=1)
     )
    )) {isSetParameters=true}
  ]
  rendererClasses [&NiagaraSpriteRendererProperties]
  renderers [
   {rendererClass=&NiagaraSpriteRendererProperties}
   {rendererClass=&NiagaraSpriteRendererProperties, rendererIndex=1}
   {rendererClass=&NiagaraSpriteRendererProperties, rendererIndex=2}
   {rendererClass=&NiagaraSpriteRendererProperties, rendererIndex=3}
  ]
 }
}
summary {
 systemName=NS_Pickaxe_SilverCanine_Trail
 emitters [
  {emitterName=TrailSharp, rendererClasses [&NiagaraRibbonRendererProperties]}
  {emitterName=Trail, rendererClasses [&NiagaraRibbonRendererProperties]}
  {emitterName=Dark_Smoke001, rendererClasses [&NiagaraRibbonRendererProperties]}
  {emitterName=Variant_Sprites, simTarget=GPUComputeSim, rendererClasses [&NiagaraSpriteRendererProperties]}
 ]
 userVariables [
  User.bIsFrontEnd {type=NiagaraBool}
  User.bIsFrontEndPreview {type=NiagaraBool}
  User.RibbonIDOffset {type=NiagaraInt32}
  User.SwingSectionIndex {type=NiagaraInt32}
  User.Variant {type=NiagaraFloat, defaultValue=0, description=""}
 ]
}
